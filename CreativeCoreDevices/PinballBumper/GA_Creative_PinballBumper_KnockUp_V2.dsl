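; Gameplay ability for the Creative Pinball Bumper V2 device that knocks whatever hits it upward.
; It overrides launch events for vehicles, pawns and physics objects, forwarding each target and velocity to the bumper's matching upward launch function.
; ShowClientEffects adds the upwards pinball gameplay cue to the ability owner for client effects.
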(event EventLaunchVehicle
  (params (Instigator :type Device%20Pinball%20Bumper%20V2%20Object%20Reference) (Target :type Fort%20Athena%20Vehicle%20Object%20Reference) (Hit :type Hit%20Result%20Structure%20%28by%20ref%29) (TargetVelocity :type Vector :default "0, 0, 0"))
  (Class|DevicePinballBumperV2|LaunchVehicleUp Instigator Target TargetVelocity))

(event EventLaunchPawn
  (params (Instigator :type Device%20Pinball%20Bumper%20V2%20Object%20Reference) (Pawn :type Fort%20Pawn%20Object%20Reference) (Hit :type Hit%20Result%20Structure%20%28by%20ref%29) (PawnVelocity :type Vector :default "0, 0, 0"))
  (Class|DevicePinballBumperV2|LaunchPawnUp Instigator Pawn PawnVelocity))

(event EventLaunchPhysicsObject
  (params (Intigator :type Device%20Pinball%20Bumper%20V2%20Object%20Reference) (HitPhysicsObject :type Fort%20Physics%20Pawn%20Object%20Object%20Reference) (Hit :type Hit%20Result%20Structure) (ObjectVelocity :type Vector :default "0, 0, 0"))
  (Class|DevicePinballBumperV2|LaunchPhysicsObjectUp Intigator HitPhysicsObject ObjectVelocity))

(fn ShowClientEffects
  (Ability|AddGameplayCueToOwner (Variables|Default|GetPinballUpwardsCue)))
