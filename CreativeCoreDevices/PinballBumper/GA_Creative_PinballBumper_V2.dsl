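; Gameplay ability for the Creative Pinball Bumper V2 device that knocks away and damages whatever it targets.
; Server-forced target selection feeds LaunchTargets, which sorts vehicles, players, creatures and physics objects, filters pawns by device settings and movement state, then calls the device's launch functions with impact point and velocity.
; When targets exist or the launch is triggered, it also adds a gameplay cue, applies SetByCaller damage and notifies the device.
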(event Ability|EventActivateAbility
  (bind _asdevice_pinball_bumper_v2 (Utilities|Casting|CastToDevice_PinballBumper_V2 (Ability|GetAvatarActorfromActorInfo))
    (:then
      (Variables|Default|SetOwningActor _asdevice_pinball_bumper_v2)
      (bind _returnvalue (Ability|CommitAbility))
      (if _returnvalue
        (bind (_asynctaskproxy _targetdata) (Ability|Tasks|DoTargetSelection :TaskInstanceName "None" :bForceTargetingOnServer true)
          (:Targeted
            (if (or (>= (- (Ability|TargetData|GetDataCountfromTargetData _targetdata) 1) 0) (Variables|/CreativeCoreDevices/Device_PinballBumper_V2.Device_PinballBumper_V2_C|GetbIsTriggeredLaunch _asdevice_pinball_bumper_v2))
              (for _index (range (Ability|TargetData|GetDataCountfromTargetData _targetdata))
                (bind _blaunchedsomeone (CallFunction|LaunchTargets _targetdata _index))
                (Variables|Default|SetbLaunchedSomone (or _blaunchedsomeone (Variables|Default|GetbLaunchedSomone))))
              (CallFunction|ShowClientEffects)
              (CallFunction|DamageTargets :TargetData _targetdata :DamageToDo (Variables|/CreativeCoreDevices/Device_PinballBumper_V2.Device_PinballBumper_V2_C|GetDamage (Variables|Default|GetOwningActor)))
              (Class|DevicePinballBumperV2|OnFinishedAction (Variables|Default|GetOwningActor))
              (else
                _))
            (Ability|EndAbility))
          (:Cancelled
            _))
        (else
          _)))
    (:CastFailed
      _)))

(fn DamageTargets
  (params (TargetData :type Gameplay%20Ability%20Target%20Data%20Handle%20Structure%20%28by%20ref%29) (ApplicationTag :type Gameplay%20Tag%20Structure) (DamageToDo :type Integer))
  (bind _returnvalue (Ability|GameplayEffect|AssignTagSetbyCallerMagnitude (Ability|MakeOutgoingGameplayEffectSpec "/CreativeCoreDevices/PinballBumper/GE_Creative_PinballBumper_Damage_V2.GE_Creative_PinballBumper_Damage_V2_C") "(TagName=\"SetByCaller.Magnitude\")" (Math|Conversions|ToFloat(Integer) DamageToDo)))
  (LocalVariables|SetDamageEffectSpec _returnvalue)
  (for _array_element (Ability|TargetData|GetAllActorsfromTargetData (Variables|GetTargetData))
    (bind _asfort_athena_vehicle (Utilities|Casting|CastToFortAthenaVehicle _array_element)
      (:then
        (if (Utilities|IsArrayNonEmpty (Vehicle|GetAllPassengers _asfort_athena_vehicle))
          _))
      (:CastFailed
        _))
    (Ability|ApplyGameplayEffectSpecToTarget (LocalVariables|GetDamageEffectSpec) (Ability|TargetData|AbilityTargetDatafromActor _array_element))))

(fn LaunchPawn
  (params (Instigator :type Device%20Pinball%20Bumper%20V2%20Object%20Reference) (Pawn :type Fort%20Pawn%20Object%20Reference) (Hit :type Hit%20Result%20Structure%20%28by%20ref%29) (PawnVelocity :type Vector))
  (bind (_bblockinghit _binitialoverlap _time _distance _location _impactpoint) (Collision|BreakHitResult Hit))
  (Class|DevicePinballBumperV2|LaunchPawn Instigator Pawn _impactpoint PawnVelocity))

(fn LaunchPhysicsObject
  (params (Intigator :type Device%20Pinball%20Bumper%20V2%20Object%20Reference) (HitPhysicsObject :type Fort%20Physics%20Pawn%20Object%20Object%20Reference) (Hit :type Hit%20Result%20Structure) (ObjectVelocity :type Vector))
  (bind (_bblockinghit _binitialoverlap _time _distance _location _impactpoint) (Collision|BreakHitResult Hit))
  (Class|DevicePinballBumperV2|LaunchPhysicsObject Intigator HitPhysicsObject _impactpoint ObjectVelocity))

(fn LaunchVehicle
  (params (Instigator :type Device%20Pinball%20Bumper%20V2%20Object%20Reference) (Target :type Fort%20Athena%20Vehicle%20Object%20Reference) (Hit :type Hit%20Result%20Structure%20%28by%20ref%29) (TargetVelocity :type Vector))
  (bind (_bblockinghit _binitialoverlap _time _distance _location _impactpoint) (Collision|BreakHitResult Hit))
  (Class|DevicePinballBumperV2|LaunchVehicle Instigator Target _impactpoint TargetVelocity))

(fn LaunchTargets
  (params (TargetData :type Gameplay%20Ability%20Target%20Data%20Handle%20Structure%20%28by%20ref%29) (Index :type Integer))
  (LocalVariables|SetHitResult (Ability|TargetData|GetHitResultfromTargetData TargetData Index))
  (for _array_element (Ability|TargetData|GetActorsfromTargetData TargetData Index)
    (bind _asfort_athena_vehicle (Utilities|Casting|CastToFortAthenaVehicle _array_element)
      (:then
        (CallFunction|LaunchVehicle (Variables|Default|GetOwningActor) _asfort_athena_vehicle (LocalVariables|GetHitResult) (Transformation|GetVelocity _asfort_athena_vehicle))
        (LocalVariables|SetbInternalLaunchedSomeone true))
      (:CastFailed
        (bind _asfort_player_pawn (Utilities|Casting|CastToFortPlayerPawn _array_element)
          (:then
            (if (Math|Boolean|ANDBoolean (Variables|/CreativeCoreDevices/Device_PinballBumper_V2.Device_PinballBumper_V2_C|GetbAffectsPlayers (Variables|Default|GetOwningActor)) (not (Vehicle|IsinVehicle _asfort_player_pawn)) (not (AI|Components|NavMovement|IsFlying (Variables|/Script/Engine.Character|GetCharacterMovement _asfort_player_pawn))))
              (CallFunction|LaunchPawn (Variables|Default|GetOwningActor) _asfort_player_pawn (LocalVariables|GetHitResult) (Transformation|GetVelocity _asfort_player_pawn))
              (LocalVariables|SetbInternalLaunchedSomeone true)))
          (:CastFailed
            (bind _asfort_pawn (Utilities|Casting|CastToFortPawn _array_element)
              (:then
                (if (and (Variables|/CreativeCoreDevices/Device_PinballBumper_V2.Device_PinballBumper_V2_C|GetbAffectsCreatures (Variables|Default|GetOwningActor)) (not (AI|Components|NavMovement|IsFlying (Variables|/Script/Engine.Character|GetCharacterMovement _asfort_pawn))))
                  (CallFunction|LaunchPawn (Variables|Default|GetOwningActor) _asfort_pawn (LocalVariables|GetHitResult) (Transformation|GetVelocity _asfort_pawn))
                  (LocalVariables|SetbInternalLaunchedSomeone true)))
              (:CastFailed
                (bind _asfort_physics_pawn_object (Utilities|Casting|CastToFortPhysicsPawnObject _array_element))
                (CallFunction|LaunchPhysicsObject (Variables|Default|GetOwningActor) _asfort_physics_pawn_object (LocalVariables|GetHitResult) (Transformation|GetVelocity _asfort_physics_pawn_object))
                (LocalVariables|SetbInternalLaunchedSomeone true))))))))
  (return (LocalVariables|GetbInternalLaunchedSomeone)))

(fn ShowClientEffects
  (Ability|AddGameplayCueToOwner (Variables|Default|GetPinballSidewaysCue)))
